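# Conversation database for the NPC KoboLobo, defining its greeting and the branches that follow it.
# The entry point feeds one Speech node whose default line can be replaced by contextual messages for a first-ever meeting, match conversation count checks, or two specific player outfits.
# First-meeting and outfit reactions use named priority curve rows.
# The speech then fans out through link nodes to the shared NPCLibrary quests, BasicServices and KoboLobo dialogue entries.
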ConversationDatabase /NPCLibrary/NPCs/KoboLobo/NPCLibrary_Conversation_KoboLobo
entry_points [{tag=Conversation.Entry.KoboLobo, destinations [#1]}]
nodes [
 ConversationEntryPointNode_7 {
  id=#1
  class=ConversationEntryPointNode
  outputs [#2]
  properties {entryTag {tagName=Conversation.Entry.KoboLobo}}
 }
 Speech_C_4 {
  id=#2
  class=Speech_C
  outputs [#3, #4, #5]
  properties {
   nodeName=""
   generalConfig {
    defaultMessage="Feeling the need to HOWL!"
    contextualMessages [
     {
      message="*sniffs vigorously* Ah, someone new. 'Sup!"
      contextRequirements [{requirement {class=FortControllerRequirement_IsFirstEverConversationWithNPC}}]
      priority {curve {rowName=Tandem.Speech.Priority.FirstEverInteraction}}
     }
     {
      message="Feeling the need to HOWL!"
      contextRequirements [{requirement {class=FortControllerRequirement_CheckNPCMatchConversationCount, properties {nPCUniqueIdTag {tagName=AISpawnerData.Type.Tandem.KoboLobo}}}}]
      priority {curve {rowName=None}}
     }
     {
      message=*fist bump* Wanna shoot some hoops?
      contextRequirements [{requirement {class=FortControllerRequirement_CheckNPCMatchConversationCount, properties {nPCUniqueIdTag {tagName=AISpawnerData.Type.Tandem.KoboLobo}}}}]
      priority {curve {rowName=None}}
     }
     {
      message=I run with a wild pack.
      contextRequirements [{requirement {class=FortControllerRequirement_CheckNPCMatchConversationCount, properties {nPCUniqueIdTag {tagName=AISpawnerData.Type.Tandem.KoboLobo}}}}]
      priority {curve {rowName=None}}
     }
     {
      message="*hackles rise* Move along, knight. Nothing to see here."
      contextRequirements [{requirement {class=FortControllerRequirement_HasCID, properties {allowedCIds [/BRCosmetics/Athena/Items/Cosmetics/Characters/Character_BucketKick]}}}]
      priority {curve {rowName=Tandem.Speech.Priority.OutfitReaction}}
     }
     {
      message=*squints* Keep those hands away from my luxurious mane...
      contextRequirements [{requirement {class=FortControllerRequirement_HasCID, properties {allowedCIds [/BRCosmetics/Athena/Items/Cosmetics/Characters/Character_ShrimpStroll]}}}]
      priority {curve {rowName=Tandem.Speech.Priority.OutfitReaction}}
     }
    ]
   }
  }
 }
 ConversationLinkNode_2 {id=#5, class=ConversationLinkNode, properties {remoteEntryTag {tagName=Conversation.Entry.KoboLobo.Dialogue}}}
 ConversationLinkNode_0 {id=#4, class=ConversationLinkNode, properties {remoteEntryTag {tagName=Conversation.Entry.BasicServices}}}
 ConversationLinkNode_1 {id=#3, class=ConversationLinkNode, properties {remoteEntryTag {tagName=Conversation.Entry.NPCLibrary.Quests}}}
]
